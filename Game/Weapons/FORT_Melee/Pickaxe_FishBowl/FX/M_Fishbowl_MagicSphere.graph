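# Additive, unlit, two-sided Niagara mesh-particle material for the magic sphere effect on the Fishbowl pickaxe.
# Two noise-distorted, panning plasma coil samples are lighten-blended, masked by scrolling noise, then added to a blue-tinted core with noise-modulated view-angle fade.
# Both layers are boosted to HDR, with optional particle color tint via UsePC.
# An alpha mask combines Fresnel-reduced coils with the core fade, minus a dynamic parameter.
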Material /Game/Weapons/FORT_Melee/Pickaxe_FishBowl/FX/M_Fishbowl_MagicSphere
graph {
 properties {
  bUsedWithNiagaraMeshParticles=true
  blendMode=BLEND_Additive
  pixelDepthOffsetMode=PDOM_Legacy
  shadingModel=MSM_Unlit
  twoSided=true
  usedShadingModels=Unlit
 }
 nodes [
  n1 = TextureCoordinate()
  n2 = Constant() {r=1}
  n3 = Multiply(A=TextureCoordinate(), B=Constant() {r=3})
  n4 = ComponentMask(None=TextureSample(UVs=Panner(Coordinate=@n3) {speedX=0.5, speedY=0.5}) {texture=/Game/Effects/Fort_Effects/Emotes/UltraLight/Texture/T_SoftNoise_BC7}.RGB) {g=true, r=true}
  n5 = Constant() {r=5}
  n6 = Panner(Coordinate=Add(A=Multiply(A=@n1, B=@n2), B=Divide(A=@n4, B=@n5))) {speedX=-1, speedY=-0.5}
  n7 = Multiply(A=@n6) {constB=2}
  n8 = Multiply(A=TextureCoordinate(), B=Constant() {r=1})
  n9 = TextureCoordinate()
  n10 = Constant() {r=3}
  n11 = TextureSample(UVs=Panner(Coordinate=Multiply(A=@n9, B=@n10)) {speedX=-0.5, speedY=0.5}) {samplerType=SAMPLERTYPE_LinearGrayscale, texture=/Game/Environments/Hexmap/HexClouds/T_HexmapFog_01_Height02}
  n12 = Divide(A=ComponentMask(None=@n11.RGB) {g=true, r=true}, B=Constant() {r=5})
  n13 = TextureSample(UVs=Panner(Coordinate=Add(A=@n8, B=@n12)) {speedX=1, speedY=1}) {texture=/Game/Effects/Fort_Effects/Textures/Tile/T_Noise_02}
  n14 = Constant() {r=5}
  n15 = Multiply(
   A=MaterialFunctionCall!(
    Blend=TextureSample(UVs=@n6) {samplerType=SAMPLERTYPE_LinearGrayscale, texture=/Game/Effects/Fort_Effects/Emotes/UltraLight/Texture/T_PlasmaCoils}.R
    Base=TextureSample(UVs=@n7) {samplerType=SAMPLERTYPE_LinearGrayscale, texture=&T_PlasmaCoils}.R
   ) {materialFunction=/Engine/Functions/Engine_MaterialFunctions03/Blends/Blend_Lighten}.Result
   B=Saturate(None=Multiply(A=@n13.RGB, B=@n14))
  )
  n16 = Constant() {r=10}
  n17 = ScalarParameter() {defaultValue=200, parameterName=NoiseBrightness}
  n18 = TextureSample() {texture=&T_SoftNoise_BC7}
  n19 = Constant() {r=0.6}
  n20 = MaterialFunctionCall("FadeSensitivity (optional)"=Multiply(A=LinearInterpolate(Alpha=@n11.R) {constA=0.1}, B=Constant() {r=50})) {materialFunction=&FadeBasedOnViewAngle}
  n21 = Multiply(A=Multiply(A=Desaturation(None=@n18.RGB, Fraction=@n19), B=Constant3Vector() {constant=<rgba 2.2987 3.4694 5 1>}), B=@n20.Result)
  n22 = ScalarParameter() {defaultValue=40, parameterName=CoreBrightness}
  n23 = Add(
   A=MaterialFunctionCall(LDR Input=@n15, HDR Dynamic Range=@n16, HDR Tint=Multiply(A=@n17, B=@n15)) {materialFunction=&DeriveHDRfromLDR}.RemappedRange
   B=MaterialFunctionCall(LDR Input=@n21, HDR Dynamic Range=@n16, HDR Intensity Multiplier=Multiply(A=@n21, B=@n22)) {materialFunction=&DeriveHDRfromLDR}.RemappedRange
  )
  n24 = ParticleColor()
  MaterialFunctionCall(Emissive=LinearInterpolate(A=@n23, B=Multiply(A=@n23, B=@n24.RGB), Alpha=ScalarParameter() {parameterName=UsePC})) {materialFunction=&MF_FX_EmissiveAdjustments}
  n25 = Fresnel() {baseReflectFraction=0, exponent=3}
  n26 = Saturate(None=Subtract(A=Power(Base=@n15), B=Abs(None=@n25)))
  Saturate(None=Subtract(A=Add(A=@n26, B=@n20.Result), B=DynamicParameter() {defaultValue=<rgba 0 1 1 1>}.Param1))
 ]
}
